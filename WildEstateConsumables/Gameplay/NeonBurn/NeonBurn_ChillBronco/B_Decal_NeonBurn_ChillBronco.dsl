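; Decal actor Blueprint for the NeonBurn Chill Bronco consumable, named for a hit effect.
; Its construction script creates a dynamic instance of the hit decal material variant without caustic motion.
; It stores that instance in a Decal_MID variable and applies it to the Decal component.
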(fn ConstructionScript
  (bind _returnvalue (Rendering|Material|CreateDynamicMaterialInstance "/WildEstateConsumables/Gameplay/NeonBurn/NeonBurn_ChillBronco/Assets/MI_Decal_NeonBurn_ChillBronco_Hit_NoCuasticMotion.MI_Decal_NeonBurn_ChillBronco_Hit_NoCuasticMotion" "None" "None" :!class /Script/Engine.KismetMaterialLibrary))
  (bind _output_get (Variables|Default|SetDecal_MID _returnvalue))
  (Rendering|Components|Decal|SetDecalMaterial (Variables|Decal|GetDecal) _output_get))
